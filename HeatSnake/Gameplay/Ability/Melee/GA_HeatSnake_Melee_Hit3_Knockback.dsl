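; Knockback gameplay ability for the third hit of the HeatSnake melee, extending an inherited launch behavior.
; Its only logic overrides DoLaunch, passing the velocity to the parent implementation.
; It then resolves the avatar as a Fort player pawn and, if valid, uncrouches it so launched players do not stay crouched.
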(event EventDoLaunch
  (params (Vel :type Vector :default "0, 0, 0"))
  ;[0.01011 0.1582 0.1719] DoLaunch: Uncrouch player on launch.
  (|CallParentFunction Vel)
  (bind _outfortplayerpawn (CallFunction|GetAvatarasValidFortPlayerPawn)
    (:ValidFortPlayerPawn
      (Character|UnCrouch _outfortplayerpawn))))
